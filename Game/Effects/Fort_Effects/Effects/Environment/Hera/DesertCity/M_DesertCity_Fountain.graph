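# Translucent, lit, two-sided Niagara mesh-particle material for the water jets of the DesertCity fountain.
# Panned noise warps a GeneratedBand, which HeightLerp erodes with optional droplet texture to shape color, opacity and roughness.
# Particle dynamic parameters drive noise, dissolve, a growing half-pill mask and tapered SplineThicken world position offset.
# Emissive combines height-faded particle light with optional time-of-day color; opacity adds quality-switched depth fade and near/far fading.
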Material /Game/Effects/Fort_Effects/Effects/Environment/Hera/DesertCity/M_DesertCity_Fountain
graph {
 properties {
  bAutomaticallySetUsageInEditor=false
  bUsedWithNiagaraMeshParticles=true
  blendMode=BLEND_Translucent
  translucencyLightingMode=TLM_VolumetricPerVertexDirectional
  twoSided=true
  usedShadingModels=Default Lit
 }
 nodes [
  Constant() {desc=Feather, r=0.2}
  n1 = DynamicParameter() {defaultValue=<rgba 1 1 1 0>, paramNames [Noise U Freq Scalar, Noise V Freq Scalar, Noise Pan Offset, Noise Dissolve]}
  n2 = Add(A=TextureCoordinate(), B=ParticleRandom())
  n3 = ComponentMask(None=@n2) {r=true}
  n4 = ScalarParameter() {defaultValue=0.2, group=Noise, parameterName=Noise U Frequency, sortPriority=1}
  n5 = DynamicParameter() {defaultValue=<rgba 1 1 1 0>, paramNames [Noise U Freq Scalar, Noise V Freq Scalar, Noise Pan Offset, Noise Dissolve]}
  n6 = Add(A=Multiply(A=Add(A=@n3) {constB=-0.5}, B=Multiply(A=@n4, B=@n5.Noise U Freq Scalar))) {constB=0.5}
  n7 = Multiply(
   A=ComponentMask(None=@n2) {g=true}
   B=Multiply(A=ScalarParameter() {defaultValue=1.2, group=Noise, parameterName=Noise V Frequency, sortPriority=2}, B=@n5.Noise V Freq Scalar)
  )
  n8 = Constant()
  n9 = ScalarParameter() {defaultValue=1, group=Noise, parameterName=Noise Speed, sortPriority=4}
  n10 = TextureObjectParameter() {
   group=Noise
   parameterName=Noise Texture
   samplerType=SAMPLERTYPE_LinearColor
   sortPriority=0
   texture=/Game/Effects/Niagara/MasterMaterials/Lightning/SplineLightning/T_SplineLightning_Noise
  }
  n11 = TextureSample(UVs=Panner(Coordinate=AppendVector(A=@n6, B=@n7), Time=@n5.Noise Pan Offset, Speed=AppendVector(A=@n8, B=@n9)), Tex=@n10) {
   samplerSource=SSM_Wrap_WorldGroupSettings
   samplerType=SAMPLERTYPE_LinearColor
   texture=&T_SplineLightning_Noise
  }
  n12 = Saturate(None=AppendVector(A=@n11.R, B=@n11.R))
  n13 = DynamicParameter() {parameterIndex=1, defaultValue=<rgba 1 1 1 0.1>, paramNames [Width Scalar, Growth, Amplitude Scalar, Cap Feather]}
  n14 = Multiply(
   A=Multiply(A=Subtract(A=@n12) {constB=0.5}) {constB=2}
   B=Multiply(A=ScalarParameter() {defaultValue=0.2, group=Noise, parameterName=Noise U Amplitude, sliderMax=1, sortPriority=3}, B=@n13.Amplitude Scalar)
  )
  n15 = MaterialFunctionCall!(
   Width=ScalarParameter() {defaultValue=0.25, group=Opacity, parameterName=Band Width}
   Sharpness=Constant()
   Offset
   Direction Switch=StaticBool() {value=true}
   Compare
   Input Coordinates=Add(A=TextureCoordinate(), B=@n14)
  ) {materialFunction=/Engine/Functions/Engine_MaterialFunctions03/Procedurals/GeneratedBand}
  n16 = AppendVector(
   A=Multiply(A=NamedRerouteUsage(), B=ScalarParameter() {defaultValue=4, group=Erode, parameterName=Droplets U Scalar})
   B=Multiply(A=NamedRerouteUsage(), B=ScalarParameter() {defaultValue=4, group=Erode, parameterName=Droplets V Scalar})
  )
  n17 = DynamicParameter() {defaultValue=<rgba 1 1 1 0>, paramNames [Noise U Freq Scalar, Noise V Freq Scalar, Noise Pan Offset, Noise Dissolve]}
  n18 = AppendVector(A=Constant(), B=ScalarParameter() {defaultValue=1, group=Noise, parameterName=Noise Speed, sortPriority=4})
  n19 = TextureSampleParameter2D(UVs=Add(A=Panner(Coordinate=@n16, Time=@n17.Noise Pan Offset, Speed=@n18), B=ParticleRandom())) {
   group=Erode
   parameterName=Droplets Texture
   samplerType=SAMPLERTYPE_LinearGrayscale
   texture=/Game/Global/Textures/Default/PointSplatter/T_GlowAlgae
  }
  n20 = Multiply(A=@n15.Result, B=Saturate(None=@n19.R)) {constA=1}
  n21 = MaterialFunctionCall(
   Transition Phase=@n1.Noise Dissolve
   Height Texture=LinearInterpolate(A=OneMinus(None=@n15.Result), B=OneMinus(None=@n20), Alpha=ScalarParameter() {group=Erode, parameterName=Droplets Erode Switch}) {constA=1}
   Contrast=ScalarParameter() {group=Erode, parameterName=Erode Contrast, sortPriority=0}
  ) {materialFunction=&HeightLerp}
  n22 = Multiply(
   A=Multiply(A=Multiply(A=@n15.Result, B=@n21.Results), B=NamedRerouteUsage())
   B=ScalarParameter() {defaultValue=1, group=Opacity, parameterName=Opacity Scalar, sortPriority=0}
  )
  n23 = LinearInterpolate(
   A=VectorParameter() {group=Color, parameterName=Outer Color, defaultValue=<rgba 0.65 1 1 1>}.RGB
   B=VectorParameter() {group=Color, parameterName=Inner Color, defaultValue=<rgba 0.2573 0.7325 1 1>}.RGB
   Alpha=@n22
  )
  n24 = ParticleColor()
  Multiply(
   A=ScalarParameter() {group=Color, parameterName=Emissive Water Mult}
   B=NamedRerouteDeclaration(None=Multiply(A=@n23, B=@n24.RGB)) {name=Base Color, nodeColor=<rgba 0 1 0.9337 1>}
  )
  NamedRerouteDeclaration(None=MaterialFunctionCall!(
    NearFadeLength=ScalarParameter() {defaultValue=1000, group=CloseFarFader, parameterName="NearFadeLength (S)"}
    NearFadeOffset=ScalarParameter() {defaultValue=300, group=CloseFarFader, parameterName="NearFadeOffset (S)"}
    NearFadeCurve
    NearFadeStartingScale
    NearFadeEndingScale
    FarFadeLength=ScalarParameter() {defaultValue=25000, group=CloseFarFader, parameterName="FarFadeLength (S)"}
    FarFadeOffset=ScalarParameter() {defaultValue=12500, group=CloseFarFader, parameterName="FarFadeOffset (S)"}
    FarFadeCurve
    FarFadeEndingScale
    Opacity Near Distance Bias
    Opacity Far Distance Bias=ScalarParameter() {defaultValue=-2, group=CloseFarFader, parameterName="Opacity Far Distance Bias (S)"}
   ) {materialFunction=/Game/Effects/Fort_Effects/Materials/_Functions/MF_SmoothLod}.Opacity) {name=CloseFarOpacity, nodeColor=<rgba 0.2656 0 1 1>}
  n25 = TextureCoordinate()
  NamedRerouteDeclaration(None=ComponentMask(None=@n25) {r=true}) {name=TexCoord.u, nodeColor=<rgba 1 0 0.9141 1>}
  n26 = DynamicParameter() {parameterIndex=1, defaultValue=<rgba 1 1 1 0.1>, paramNames [Width Scalar, Growth, Amplitude Scalar, Cap Feather]}
  n27 = Constant() {desc=PillRadius, r=4}
  n28 = Multiply(A=NamedRerouteUsage(), B=@n27)
  n29 = DynamicParameter() {parameterIndex=1, defaultValue=<rgba 1 1 1 0.1>, paramNames [Width Scalar, Growth, Amplitude Scalar, Cap Feather]}
  n30 = MaterialFunctionCall(Input=@n29.Growth, Input Low=Constant(), Input High=Constant() {r=2}, Target Low=Multiply(A=@n27) {constB=-1}, Target High=@n27) {materialFunction=&RemapValueRange}
  n31 = Subtract(A=OneMinus(None=Add(A=@n28, B=@n30.Result)), B=Constant() {desc=TopV, r=0.3})
  n32 = Abs(None=Subtract(A=NamedRerouteUsage()) {constB=0.5})
  n33 = AppendVector(A=@n32, B=@n31)
  n34 = Constant() {desc=HalfWidth, r=0.5}
  n35 = Min(
   A=Subtract(A=MaterialFunctionCall(Vector 2=@n33) {materialFunction=/Engine/Functions/Engine_MaterialFunctions02/Utility/VectorLength}.V2 Length, B=@n34) {constB=0.105}
   B=Max(A=Subtract(A=@n32, B=@n34), B=@n31)
  )
  NamedRerouteDeclaration(None=OneMinus(None=SmoothStep(Max=@n26.Cap Feather, Value=@n35))) {name=HalfPillMask, nodeColor=<rgba 1 1 1 1>}
  NamedRerouteDeclaration(None=ComponentMask(None=@n25) {g=true}) {name=TexCoord.v, nodeColor=<rgba 1 0 0.9141 1>}
  NamedRerouteDeclaration(None=MaterialFunctionCall!(
    Height Map=@n10
    Normal Map Intensity=ScalarParameter() {defaultValue=2, group=Surface Properties, parameterName=Normal Map Intensity, sortPriority=1}
    Height Map UV Offset
    Coordinates=@n14
    Height Map Channel Selector
   ) {materialFunction=/Engine/Functions/Engine_MaterialFunctions03/Procedurals/NormalFromHeightmap}.Result) {name=FinalNormal, nodeColor=<rgba 0.1497 1 0.26 1>}
  n36 = DynamicParameter() {parameterIndex=1, defaultValue=<rgba 1 1 1 0.1>, paramNames [Width Scalar, Growth, Amplitude Scalar, Cap Feather]}
  n37 = DynamicParameter() {parameterIndex=1, defaultValue=<rgba 1 1 1 0.1>, paramNames [Width Scalar, Growth, Amplitude Scalar, Cap Feather]}
  n38 = Divide(
   A=Subtract(A=NamedRerouteUsage(), B=OneMinus(None=@n37.Growth))
   B=Max(A=ScalarParameter() {defaultValue=0.5, group=WPO, parameterName=Feather, sliderMax=1, sortPriority=0}) {constB=0.001}
  )
  n39 = Multiply(
   A=Multiply(A=ScalarParameter() {defaultValue=2, group=WPO, parameterName=Width, sortPriority=0}, B=@n36.Width Scalar)
   B=LinearInterpolate(
    A=ScalarParameter() {defaultValue=0.5, group=WPO, parameterName=Width Tip, sliderMax=1, sortPriority=2}
    B=ScalarParameter() {defaultValue=1, group=WPO, parameterName=Width Base, sliderMax=1, sortPriority=1}
    Alpha=Saturate(None=@n38)
   )
  )
  NamedRerouteDeclaration(None=MaterialFunctionCall(WidthBase=@n39, WidthTip=@n39) {materialFunction=&SplineThicken}.WorldPositionOffset) {name=WPO, nodeColor=<rgba 1 0 0 1>}
  NamedRerouteDeclaration(None=LinearInterpolate(Alpha=@n22) {constA=0.05, constB=0.4}) {name=Roughness, nodeColor=<rgba 0.8281 0 1 1>}
  n40 = Multiply(A=@n24.A, B=@n22)
  n41 = DepthFade(Opacity=@n40, FadeDistance=ScalarParameter() {defaultValue=10, group=Opacity, parameterName=FadeDistance, sortPriority=1})
  NamedRerouteDeclaration(None=Multiply(A=QualitySwitch(Default=@n41, Low=@n40, High=@n41, Medium=@n40, Epic=@n41), B=NamedRerouteUsage())) {name=Opacity, nodeColor=<rgba 0.0547 1 0 1>}
  n42 = MaterialFunctionCall(
   Input=WorldPosition().Z
   Input Low=ScalarParameter() {group=Color, parameterName=Min World Light Emissive, sortPriority=4}
   Input High=ScalarParameter() {defaultValue=400, group=Color, parameterName=Max World Light Emissive, sortPriority=5}
   Target Low=Constant()
   Target High=Constant() {r=1}
  ) {materialFunction=&RemapValueRange}
  n43 = Multiply(
   A=DynamicParameter() {parameterIndex=2, defaultValue=<rgba 5 3 0.5 1>, paramNames [R,G,B,A]}.RGB
   B=Power(Base=OneMinus(None=@n42.Result), Exp=ScalarParameter() {defaultValue=2, group=Color, parameterName=Emissive Light Exponent, sortPriority=4})
  )
  n44 = StaticSwitchParameter(
   True=MaterialFunctionCall(
    Color_Morning=VectorParameter() {group=Color, parameterName=Emissive Morning, sortPriority=6, defaultValue=<rgba 0 0 0 1>}.RGB
    Color_Day=VectorParameter() {group=Color, parameterName=Emissive Day, sortPriority=7, defaultValue=<rgba 0 0 0 1>}.RGB
    Color_Evening=VectorParameter() {group=Color, parameterName=Emissive Evening, sortPriority=8, defaultValue=<rgba 0 0 0 1>}.RGB
    Color_Night=VectorParameter() {group=Color, parameterName=Emissive Night, sortPriority=9, defaultValue=<rgba 0 0 0 1>}.RGB
   ) {materialFunction=&Color_Time_of_Day}.Result
   False=Constant3Vector()
  ) {group=Color, parameterName=Toggle Emissive Water, sortPriority=5}
  NamedRerouteDeclaration(None=MaterialFunctionCall(Emissive=Add(A=@n43, B=@n44)) {materialFunction=&MF_FX_EmissiveAdjustments}.Result) {name=Emissive, nodeColor=<rgba 1 0.8802 0 1>}
  NamedRerouteUsage()
  NamedRerouteUsage()
  NamedRerouteUsage()
  NamedRerouteUsage()
  NamedRerouteUsage()
  NamedRerouteUsage()
  ScalarParameter() {defaultValue=0.25, group=Surface Properties, parameterName=Specular, sortPriority=0}
 ]
}
