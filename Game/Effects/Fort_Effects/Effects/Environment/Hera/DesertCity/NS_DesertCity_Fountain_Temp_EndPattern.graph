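# Niagara system for the temporary end pattern of the DesertCity fountain effect, drawn with four GPU mesh-particle emitters.
# LineMoving and LineMovingForward sweep spawn points side to side by loop-age curves, the latter also stepping forward.
# Rings bursts three widening circles from a radius array indexed by loop count, and Spiral traces a growing flattened torus knot.
# All share age-driven dynamic material parameters with decelerating noise panning; no emitter shown reads the user pattern, position or delay arrays.
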NiagaraSystem /Game/Effects/Fort_Effects/Effects/Environment/Hera/DesertCity/NS_DesertCity_Fountain_Temp_EndPattern
dependencies {
 usedDataInterfaces [&NiagaraDataInterfaceCurve]
 usedDynamicInputs [
  &Ceil_Float
  &Divide_Float
  &MakeFloatFromInt
  /Niagara/DynamicInputs/Math/GetInt32ArrayCount
 ]
 usedModules [&SystemState]
}
emitters {
 LineMoving {
  simTarget=GPUComputeSim
  emitterUpdateScript [
   &EmitterState(Life Cycle Mode=::NewEnumerator1, Loop Behavior=::NewEnumerator1, UseLoopDelay=false)
   &SpawnRate(SpawnRate=25, Use Spawn Probability=false)
  ]
  particleSpawnScript [
   &InitializeParticle(
    Color Mode=::NewEnumerator0
    Lifetime=4
    Mesh Scale=<1 1 3>
    Mesh Scale Mode=::NewEnumerator4
    Position Mode=::NewEnumerator1
    UsePositionOffset=false
    Write Material Random=true
   )
   /SetVariables() {isSetParameters=true, module=SetVariables_#1}
   &CameraOffset(Camera Offset Amount=25)
   &SystemLocation(
    MaskSpawnGroup=false
    Offset=&MakeVector(
     X=0
     Y=&FloatFromCurve(
      FloatCurve=DataInterface() {propertyValues {Curve {
         defaultValue=1
         keys [
          {interpMode=RCIM_Cubic, tangentMode=RCTM_User, time=0, value=-1}
          {interpMode=RCIM_Cubic, tangentMode=RCTM_User, time=0.125, value=1}
          <at 0.25 -0.8447 RCIM_Cubic>
          {arriveTangent=-0.6213, interpMode=RCIM_Cubic, leaveTangent=-0.6213, time=0.375, value=1}
          <at 0.5 -1 RCIM_Cubic>
          <at 0.625 1 RCIM_Cubic>
          <at 0.75 -1 RCIM_Cubic>
          <at 0.875 1 RCIM_Cubic>
          {tangentMode=RCTM_User, time=1, value=-1}
         ]
        }}}
      CurveIndex=$Emitter.NormalizedLoopAge
      Scale Curve=1000
     )
     Z=0
    )
   )
   &ShapeLocation(MaskSpawnGroup=false, Override Seed=false, Sphere Radius=25, UseOwnerScale=true)
  ]
  particleUpdateScript [
   &ParticleState()
   &ScaleMeshSize(Scale Factor=&VectorFromFloat(Value=3))
   /SetVariables(Particles.NormalizedAge:NiagaraFloat=&RemapRange(Input Value To Remap=$Particles.Age, Input Min=$Particles.Delay, Input Max=$Particles.Lifetime, Output Min=0, Output Max=1)) {isSetParameters=true, module=SetVariables_#2}
   /SetVariables(Particles.NoisePan:NiagaraFloat=&Add_Float(
     A=$Particles.NoisePan
     B=&Multiply_Float(
      A=&FloatFromCurve(
       FloatCurve=DataInterface() {propertyValues {Curve {defaultValue=1, keys [{tangentMode=RCTM_User, time=0, value=1}, {tangentMode=RCTM_User, time=1, value=0}]}}}
       Scale Curve=0.5
      )
      B=$Engine.DeltaTime
     )
    )) {isSetParameters=true, module=SetVariables_#3}
   &DynamicMaterialParameters(
    Index 0 Param 1=1
    Index 0 Param 2=&Lerp_Float(
     A=0.4
     B=2
     Alpha=&FloatFromCurve(FloatCurve=DataInterface() {propertyValues {Curve {defaultValue=0, keys [{tangentMode=RCTM_User, time=0, value=0}, {tangentMode=RCTM_User, time=1, value=1}]}}})
    )
    Index 0 Param 3=$Particles.NoisePan
    Index 0 Param 4=&FloatFromCurve(FloatCurve=DataInterface() {propertyValues {Curve {defaultValue=0, keys [{tangentMode=RCTM_User, time=0, value=0}, {tangentMode=RCTM_User, time=1, value=1}]}}})
    Index 1 Param 1=&Lerp_Float(
     A=1
     B=5
     Alpha=&FloatFromCurve(FloatCurve=DataInterface() {propertyValues {Curve {
         defaultValue=0
         keys [
          {interpMode=RCIM_Cubic, tangentMode=RCTM_User, time=0, value=0}
          {arriveTangent=2.4939, interpMode=RCIM_Cubic, leaveTangent=2.4939, tangentMode=RCTM_User, time=1, value=1}
         ]
        }}})
    )
    Index 1 Param 2=&FloatFromCurve(FloatCurve=DataInterface() {propertyValues {Curve {
        defaultValue=0
        keys [
         {arriveTangent=5.795, interpMode=RCIM_Cubic, leaveTangent=5.795, tangentMode=RCTM_User, time=0, value=0}
         {arriveTangent=1.8993, interpMode=RCIM_Cubic, leaveTangent=1.8993, tangentMode=RCTM_User, time=0.1525, value=0.8805}
         {arriveTangent=-0.8363, interpMode=RCIM_Cubic, leaveTangent=-0.8363, tangentMode=RCTM_User, time=0.433, value=0.913}
         {arriveTangent=0.1838, interpMode=RCIM_Cubic, leaveTangent=0.1838, tangentMode=RCTM_User, time=1, value=0}
        ]
       }}})
    Module.Index0Float0Write=true
    Module.Index0Float1Write=true
    Module.Index0Float2Write=true
    Module.Index0Float3Write=true
    Module.Index1Float0Write=true
    Module.Index1Float1Write=true
    Module.Index1Float2Write=false
    Module.Index1Float3Write=false
    Write Parameter Index 1=::NewEnumerator1
   )
  ]
  rendererClasses [&NiagaraMeshRendererProperties]
  renderers [
   {rendererClass=&NiagaraMeshRendererProperties, rendererIndex=0}
   {rendererClass=&NiagaraMeshRendererProperties, rendererIndex=1}
  ]
 }
 LineMovingForward {
  simTarget=GPUComputeSim
  emitterUpdateScript [
   &EmitterState(Life Cycle Mode=::NewEnumerator1, Loop Behavior=::NewEnumerator1, Loop Duration=2, UseLoopDelay=false)
   &SpawnRate(SpawnRate=32, Use Spawn Probability=false)
  ]
  particleSpawnScript [
   &InitializeParticle(
    Color Mode=::NewEnumerator0
    Lifetime=4
    Mesh Scale=<1 1 3>
    Mesh Scale Mode=::NewEnumerator4
    Position Mode=::NewEnumerator1
    UsePositionOffset=false
    Write Material Random=true
   )
   /SetVariables() {isSetParameters=true, module=SetVariables_#1}
   &CameraOffset(Camera Offset Amount=25)
   &SystemLocation(
    MaskSpawnGroup=false
    Offset=&MakeVector(
     X=&FloatFromCurve(
      FloatCurve=DataInterface() {propertyValues {Curve {
         defaultValue=1
         keys [<at 0 -1 RCIM_Constant>,<at 0.125 -0.75 RCIM_Constant>,<at 0.25 -0.5 RCIM_Constant>,<at 0.375 -0.25 RCIM_Constant>,<at 0.5 0 RCIM_Constant>,<at 0.625 0.25 RCIM_Constant>,<at 0.75 0.5 RCIM_Constant>,<at 0.875 0.75 RCIM_Constant>,<at 1 1 RCIM_Constant>]
        }}}
      CurveIndex=$Emitter.NormalizedLoopAge
      Scale Curve=500
     )
     Y=&FloatFromCurve(
      FloatCurve=DataInterface() {propertyValues {Curve {
         defaultValue=1
         keys [
          {interpMode=RCIM_Cubic, tangentMode=RCTM_User, time=0, value=-1}
          {interpMode=RCIM_Cubic, tangentMode=RCTM_User, time=0.125, value=1}
          <at 0.25 -0.8447 RCIM_Cubic>
          {arriveTangent=-0.6213, interpMode=RCIM_Cubic, leaveTangent=-0.6213, time=0.375, value=1}
          <at 0.5 -1 RCIM_Cubic>
          <at 0.625 1 RCIM_Cubic>
          <at 0.75 -1 RCIM_Cubic>
          <at 0.875 1 RCIM_Cubic>
          {tangentMode=RCTM_User, time=1, value=-1}
         ]
        }}}
      CurveIndex=$Emitter.NormalizedLoopAge
      Scale Curve=1000
     )
     Z=0
    )
   )
   &ShapeLocation(MaskSpawnGroup=false, Override Seed=false, Sphere Radius=25, UseOwnerScale=true)
  ]
  particleUpdateScript [
   &ParticleState()
   &ScaleMeshSize(Scale Factor=&VectorFromFloat(Value=3))
   /SetVariables(Particles.NormalizedAge:NiagaraFloat=&RemapRange(Input Value To Remap=$Particles.Age, Input Min=$Particles.Delay, Input Max=$Particles.Lifetime, Output Min=0, Output Max=1)) {isSetParameters=true, module=SetVariables_#2}
   /SetVariables(Particles.NoisePan:NiagaraFloat=&Add_Float(
     A=$Particles.NoisePan
     B=&Multiply_Float(
      A=&FloatFromCurve(
       FloatCurve=DataInterface() {propertyValues {Curve {defaultValue=1, keys [{tangentMode=RCTM_User, time=0, value=1}, {tangentMode=RCTM_User, time=1, value=0}]}}}
       Scale Curve=0.5
      )
      B=$Engine.DeltaTime
     )
    )) {isSetParameters=true, module=SetVariables_#3}
   &DynamicMaterialParameters(
    Index 0 Param 1=1
    Index 0 Param 2=&Lerp_Float(
     A=0.4
     B=2
     Alpha=&FloatFromCurve(FloatCurve=DataInterface() {propertyValues {Curve {defaultValue=0, keys [{tangentMode=RCTM_User, time=0, value=0}, {tangentMode=RCTM_User, time=1, value=1}]}}})
    )
    Index 0 Param 3=$Particles.NoisePan
    Index 0 Param 4=&FloatFromCurve(FloatCurve=DataInterface() {propertyValues {Curve {defaultValue=0, keys [{tangentMode=RCTM_User, time=0, value=0}, {tangentMode=RCTM_User, time=1, value=1}]}}})
    Index 1 Param 1=&Lerp_Float(
     A=1
     B=5
     Alpha=&FloatFromCurve(FloatCurve=DataInterface() {propertyValues {Curve {
         defaultValue=0
         keys [
          {interpMode=RCIM_Cubic, tangentMode=RCTM_User, time=0, value=0}
          {arriveTangent=2.4939, interpMode=RCIM_Cubic, leaveTangent=2.4939, tangentMode=RCTM_User, time=1, value=1}
         ]
        }}})
    )
    Index 1 Param 2=&FloatFromCurve(FloatCurve=DataInterface() {propertyValues {Curve {
        defaultValue=0
        keys [
         {arriveTangent=5.795, interpMode=RCIM_Cubic, leaveTangent=5.795, tangentMode=RCTM_User, time=0, value=0}
         {arriveTangent=1.8993, interpMode=RCIM_Cubic, leaveTangent=1.8993, tangentMode=RCTM_User, time=0.1525, value=0.8805}
         {arriveTangent=-0.8363, interpMode=RCIM_Cubic, leaveTangent=-0.8363, tangentMode=RCTM_User, time=0.433, value=0.913}
         {arriveTangent=0.1838, interpMode=RCIM_Cubic, leaveTangent=0.1838, tangentMode=RCTM_User, time=1, value=0}
        ]
       }}})
    Module.Index0Float0Write=true
    Module.Index0Float1Write=true
    Module.Index0Float2Write=true
    Module.Index0Float3Write=true
    Module.Index1Float0Write=true
    Module.Index1Float1Write=true
    Module.Index1Float2Write=false
    Module.Index1Float3Write=false
    Write Parameter Index 1=::NewEnumerator1
   )
  ]
  rendererClasses [&NiagaraMeshRendererProperties]
  renderers [
   {rendererClass=&NiagaraMeshRendererProperties, rendererIndex=0}
   {rendererClass=&NiagaraMeshRendererProperties, rendererIndex=1}
  ]
 }
 Rings {
  simTarget=GPUComputeSim
  emitterUpdateScript [
   &EmitterState(Life Cycle Mode=::NewEnumerator1, Loop Behavior=::NewEnumerator2, Loop Count=3, Loop Duration=1, UseLoopDelay=false)
   &SpawnBurst_Instantaneous(Loop Count Limit=3, Spawn Count=50, Use Spawn Probability=false, UseLoopCountLimit=false)
  ]
  particleSpawnScript [
   &InitializeParticle(
    Color Mode=::NewEnumerator0
    Lifetime=4
    Mesh Scale=<1 1 3>
    Mesh Scale Mode=::NewEnumerator4
    Position Mode=::NewEnumerator1
    UsePositionOffset=false
    Write Material Random=true
   )
   /SetVariables() {isSetParameters=true, module=SetVariables_#1}
   &CameraOffset(Camera Offset Amount=25)
   &SystemLocation(MaskSpawnGroup=false, Offset=<0 -1500 0>)
   &ShapeLocation(
    MaskSpawnGroup=false
    Non Uniform Scale=<1 1 0>
    Override Seed=false
    Ring Radius=&SelectFloatFromArray(Direct Array Index=$Emitter.LoopCount, Float Selection Array=DataInterface() {propertyValues {FloatData [150,350,550]}})
    Shape Primitive=::NewEnumerator4
    UseOwnerScale=true
   )
  ]
  particleUpdateScript [
   &ParticleState()
   &ScaleMeshSize(Scale Factor=&VectorFromFloat(Value=3))
   /SetVariables(Particles.NormalizedAge:NiagaraFloat=&RemapRange(Input Value To Remap=$Particles.Age, Input Min=$Particles.Delay, Input Max=$Particles.Lifetime, Output Min=0, Output Max=1)) {isSetParameters=true, module=SetVariables_#2}
   /SetVariables(Particles.NoisePan:NiagaraFloat=&Add_Float(
     A=$Particles.NoisePan
     B=&Multiply_Float(
      A=&FloatFromCurve(
       FloatCurve=DataInterface() {propertyValues {Curve {defaultValue=1, keys [{tangentMode=RCTM_User, time=0, value=1}, {tangentMode=RCTM_User, time=1, value=0}]}}}
       Scale Curve=0.5
      )
      B=$Engine.DeltaTime
     )
    )) {isSetParameters=true, module=SetVariables_#3}
   &DynamicMaterialParameters(
    Index 0 Param 1=1
    Index 0 Param 2=&Lerp_Float(
     A=0.4
     B=2
     Alpha=&FloatFromCurve(FloatCurve=DataInterface() {propertyValues {Curve {defaultValue=0, keys [{tangentMode=RCTM_User, time=0, value=0}, {tangentMode=RCTM_User, time=1, value=1}]}}})
    )
    Index 0 Param 3=$Particles.NoisePan
    Index 0 Param 4=&FloatFromCurve(FloatCurve=DataInterface() {propertyValues {Curve {defaultValue=0, keys [{tangentMode=RCTM_User, time=0, value=0}, {tangentMode=RCTM_User, time=1, value=1}]}}})
    Index 1 Param 1=&Lerp_Float(
     A=1
     B=5
     Alpha=&FloatFromCurve(FloatCurve=DataInterface() {propertyValues {Curve {
         defaultValue=0
         keys [
          {interpMode=RCIM_Cubic, tangentMode=RCTM_User, time=0, value=0}
          {arriveTangent=2.4939, interpMode=RCIM_Cubic, leaveTangent=2.4939, tangentMode=RCTM_User, time=1, value=1}
         ]
        }}})
    )
    Index 1 Param 2=&FloatFromCurve(FloatCurve=DataInterface() {propertyValues {Curve {
        defaultValue=0
        keys [
         {arriveTangent=5.795, interpMode=RCIM_Cubic, leaveTangent=5.795, tangentMode=RCTM_User, time=0, value=0}
         {arriveTangent=1.8993, interpMode=RCIM_Cubic, leaveTangent=1.8993, tangentMode=RCTM_User, time=0.1525, value=0.8805}
         {arriveTangent=-0.8363, interpMode=RCIM_Cubic, leaveTangent=-0.8363, tangentMode=RCTM_User, time=0.433, value=0.913}
         {arriveTangent=0.1838, interpMode=RCIM_Cubic, leaveTangent=0.1838, tangentMode=RCTM_User, time=1, value=0}
        ]
       }}})
    Module.Index0Float0Write=true
    Module.Index0Float1Write=true
    Module.Index0Float2Write=true
    Module.Index0Float3Write=true
    Module.Index1Float0Write=true
    Module.Index1Float1Write=true
    Module.Index1Float2Write=false
    Module.Index1Float3Write=false
    Write Parameter Index 1=::NewEnumerator1
   )
  ]
  rendererClasses [&NiagaraMeshRendererProperties]
  renderers [
   {rendererClass=&NiagaraMeshRendererProperties, rendererIndex=0}
   {rendererClass=&NiagaraMeshRendererProperties, rendererIndex=1}
  ]
 }
 Spiral {
  simTarget=GPUComputeSim
  emitterUpdateScript [
   &EmitterState()
   &SpawnRate(
    SpawnRate=&Multiply_Float(
     A=50
     B=&FloatFromCurve(FloatCurve=DataInterface() {propertyValues {Curve {keys [<at 0 0>,<at 0.5 1>]}}}, CurveIndex=$System.NormalizedLoopAge)
    )
    Use Spawn Probability=false
   )
  ]
  particleSpawnScript [
   &InitializeParticle(
    Color Mode=::NewEnumerator0
    Lifetime=4
    Mesh Scale Mode=::NewEnumerator4
    Position Mode=::NewEnumerator1
    UsePositionOffset=false
    Write Material Random=true
   )
   /SetVariables(Particles.NoisePan=1) {isSetParameters=true, module=SetVariables_#1}
   &CameraOffset(Camera Offset Amount=25)
   &ShapeLocation(
    Enable Torus Knot Radius=true
    Handle Radius=0
    Knot Helix Loops=4
    Knot Ring Loops=2
    Large Radius=&FloatFromCurve(FloatCurve=DataInterface() {propertyValues {Curve {keys [<at 0 0>,<at 1 1>]}}}, CurveIndex=$System.NormalizedLoopAge, Scale Curve=500)
    MaskSpawnGroup=false
    Non Uniform Scale=<1 1 0>
    Shape Primitive=::NewEnumerator3
    Torus Distribution Mode:ENiagaraTorusDistributionMode=::NewEnumerator1
    Torus Knot Tube Radius=0
    Torus Mode:ENiagaraShapeTorusMode=::NewEnumerator1
    U Position=$Emitter.NormalizedLoopAge
    UseOwnerScale=true
   )
  ]
  particleUpdateScript [
   &ParticleState()
   &ScaleMeshSize(Scale Factor=&VectorFromFloat(Value=8))
   /SetVariables(Particles.NormalizedAge:NiagaraFloat=&RemapRange(Input Value To Remap=$Particles.Age, Input Min=$Particles.Delay, Input Max=$Particles.Lifetime, Output Min=0, Output Max=1)) {isSetParameters=true, module=SetVariables_#2}
   /SetVariables(Particles.NoisePan:NiagaraFloat=&Add_Float(
     A=$Particles.NoisePan
     B=&Multiply_Float(
      A=&FloatFromCurve(
       FloatCurve=DataInterface() {propertyValues {Curve {defaultValue=1, keys [{tangentMode=RCTM_User, time=0, value=1}, {tangentMode=RCTM_User, time=1, value=0}]}}}
       Scale Curve=0.5
      )
      B=$Engine.DeltaTime
     )
    )) {isSetParameters=true, module=SetVariables_#3}
   &DynamicMaterialParameters(
    Index 0 Param 1=1
    Index 0 Param 2=&Lerp_Float(
     A=0.4
     B=2
     Alpha=&FloatFromCurve(FloatCurve=DataInterface() {propertyValues {Curve {defaultValue=0, keys [{tangentMode=RCTM_User, time=0, value=0}, {tangentMode=RCTM_User, time=1, value=1}]}}})
    )
    Index 0 Param 3=$Particles.NoisePan
    Index 0 Param 4=&FloatFromCurve(FloatCurve=DataInterface() {propertyValues {Curve {defaultValue=0, keys [{tangentMode=RCTM_User, time=0, value=0}, {tangentMode=RCTM_User, time=1, value=1}]}}})
    Index 1 Param 1=&Lerp_Float(
     A=1
     B=10
     Alpha=&FloatFromCurve(FloatCurve=DataInterface() {propertyValues {Curve {
         defaultValue=0
         keys [
          {interpMode=RCIM_Cubic, tangentMode=RCTM_User, time=0, value=0}
          {arriveTangent=2.4939, interpMode=RCIM_Cubic, leaveTangent=2.4939, tangentMode=RCTM_User, time=1, value=1}
         ]
        }}})
    )
    Index 1 Param 2=&FloatFromCurve(FloatCurve=DataInterface() {propertyValues {Curve {
        defaultValue=0
        keys [
         {arriveTangent=5.795, interpMode=RCIM_Cubic, leaveTangent=5.795, tangentMode=RCTM_User, time=0, value=0}
         {arriveTangent=1.8993, interpMode=RCIM_Cubic, leaveTangent=1.8993, tangentMode=RCTM_User, time=0.1525, value=0.8805}
         {arriveTangent=-0.8363, interpMode=RCIM_Cubic, leaveTangent=-0.8363, tangentMode=RCTM_User, time=0.433, value=0.913}
         {arriveTangent=0.1838, interpMode=RCIM_Cubic, leaveTangent=0.1838, tangentMode=RCTM_User, time=1, value=0}
        ]
       }}})
    Module.Index0Float0Write=true
    Module.Index0Float1Write=true
    Module.Index0Float2Write=true
    Module.Index0Float3Write=true
    Module.Index1Float0Write=true
    Module.Index1Float1Write=true
    Module.Index1Float2Write=false
    Module.Index1Float3Write=false
    Write Parameter Index 1=::NewEnumerator1
   )
  ]
  rendererClasses [&NiagaraMeshRendererProperties]
  renderers [
   {rendererClass=&NiagaraMeshRendererProperties, rendererIndex=0}
   {rendererClass=&NiagaraMeshRendererProperties, rendererIndex=1}
  ]
 }
}
summary {
 systemName=NS_DesertCity_Fountain_Temp_EndPattern
 emitters [
  {emitterName=Spiral, simTarget=GPUComputeSim, rendererClasses [&NiagaraMeshRendererProperties]}
  {emitterName=Rings, simTarget=GPUComputeSim, rendererClasses [&NiagaraMeshRendererProperties]}
  {emitterName=LineMoving, simTarget=GPUComputeSim, rendererClasses [&NiagaraMeshRendererProperties]}
  {emitterName=LineMovingForward, simTarget=GPUComputeSim, rendererClasses [&NiagaraMeshRendererProperties]}
 ]
 userVariables [
  User.DelayArray {
   type=NiagaraDataInterfaceArrayFloat
   defaultValue {
    struct=NiagaraExt_VariableValue_DataInterface
    value {
     dataInterface=&self:NiagaraDataInterfaceArrayFloat_0
     dataInterfaceClass=&NiagaraDataInterfaceArrayFloat
    }
   }
  }
  User.LongestDelay {type=NiagaraFloat, defaultValue=0.5, description=""}
  User.PatternArray {
   type=NiagaraDataInterfaceArrayInt32
   description=""
   defaultValue {
    struct=NiagaraExt_VariableValue_DataInterface
    value {
     dataInterface=&self:NiagaraDataInterfaceArrayInt32_0
     dataInterfaceClass=&NiagaraDataInterfaceArrayInt32
    }
   }
  }
  User.PosArray {
   type=NiagaraDataInterfaceArrayPosition
   description=""
   defaultValue {
    struct=NiagaraExt_VariableValue_DataInterface
    value {
     dataInterface=&self:NiagaraDataInterfaceArrayPosition_0
     dataInterfaceClass=&NiagaraDataInterfaceArrayPosition
    }
   }
  }
 ]
}
